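; Preview actor Blueprint that shows a geometry collection as a simulated Chaos physics object.
; Its construction script sets the component's rest collection from the DataflowAsset variable without applying asset defaults.
; It then assigns the actor itself as the component's Chaos solver actor and turns on physics simulation.
; BeginPlay, overlap and Tick events are empty, so all setup happens at construction.
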(event EventBeginPlay)

(event Collision|EventActorBeginOverlap
  (params (OtherActor :type Actor%20Object%20Reference)))

(event EventTick
  (params (DeltaSeconds :type Float%20%28single-precision%29)))

(fn ConstructionScript
  (ChaosPhysics|SetRestCollection (Variables|Default|GetGeometryCollectionComponent) (Variables|Default|GetDataflowAsset) false)
  (Variables|/Script/GeometryCollectionEngine.GeometryCollectionComponent|SetChaosSolverActor self (Variables|Default|GetGeometryCollectionComponent))
  (Physics|SetSimulatePhysics (Variables|Default|GetGeometryCollectionComponent) true))
